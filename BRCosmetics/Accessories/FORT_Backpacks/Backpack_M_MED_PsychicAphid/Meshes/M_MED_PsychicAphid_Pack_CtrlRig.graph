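# Control Rig for the PsychicAphid back bling that procedurally poses the pack and its jointed limbs from the wearer's movement.
# Limb pose offsets blend in through a smoothed weight from pelvis height and forward speed, zeroed while skydiving.
# Spine-to-pelvis rotation (Yaw channel) then drives layered pack and clavicle offsets.
# Spring-driven two-bone IK and pole-vector experiments remain in the graph, disconnected from execution.
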ControlRigBlueprint /BRCosmetics/Accessories/FORT_Backpacks/Backpack_M_MED_PsychicAphid/Meshes/M_MED_PsychicAphid_Pack_CtrlRig
graphs [
 Add_3_1_ContainedGraph {nodes [
   n1 = RigVMFunctionEntryNode()
   RigVMFunctionReturnNode(Result=&RigVMFunction_MathDoubleAdd(A=&RigVMFunction_MathDoubleAdd(A=@n1.A, B=@n1.B).Result, B=@n1.C).Result)
  ]}
 RigVMFunctionLibrary {links [], nodes []}
 RigVMModel {nodes [
   n1 = &RigUnit_GetTransform(Item=(Type=Control,Name="leg_ik_l"), bInitial=False)
   DeltaFromPrevious = /Script/RigVM.RigVMFunction_DeltaFromPreviousVector(Value=&RigVMFunction_MathTransformMakeRelative(
     Parent=&RigUnit_GetTransform(Item=(Type=Bone,Name="hand_l"), bInitial=False).Transform
     Global=RigVMVariableNode() {variable=TPack}.Value
    ).Local.Translation) {is=struct RigVMUnitNode, template="DeltaFromPrevious::Execute(in Value,out Delta,out PreviousValue)"}
   Scale = /Script/RigVM.RigVMFunction_MathVectorScale(Factor=0.1, Value=@DeltaFromPrevious.Delta) {is=struct RigVMUnitNode, template="Scale::Execute(in Value,in Factor,out Result)"}
   DISPATCH_RigVMDispatch_CoreNotEquals = RigVMDispatchNode(B=(X=0,Y=0,Z=0), A=@DeltaFromPrevious.Delta) {template="DISPATCH_RigVMDispatch_CoreNotEquals(in A,in B,out Result)"}
   RigVMFunction_MathBoolToInteger = &RigVMFunction_MathBoolToInteger(Value=@DISPATCH_RigVMDispatch_CoreNotEquals.Result)
   DISPATCH_RigVMDispatch_SelectInt32 = RigVMDispatchNode(
    Values=((X=0,Y=0,Z=0),(X=0,Y=0,Z=0))
    Values.1=/Script/RigVM.RigVMFunction_MathVectorAdd(A=@Scale.Result, B=@n1.Transform.Translation) {is=struct RigVMUnitNode, template="Add::Execute(in A,in B,out Result)"}.Result
    Values.0=&RigUnit_GetTransform(Item=(Type=Control,Name="leg_ik_l"), bInitial=true).Transform.Translation
    Index=@RigVMFunction_MathBoolToInteger.Result
   ) {template="DISPATCH_RigVMDispatch_SelectInt32(in Index,in Values,out Result)"}
   SpringInterp = /Script/ControlRig.RigUnit_SpringInterpVectorV2(
    CriticalDamping=1
    Force=(X=0,Y=0,Z=0)
    Strength=1
    TargetVelocityAmount=0
    bInitializeFromTarget=false
    bUseCurrentInput=False
    Target=@DISPATCH_RigVMDispatch_SelectInt32.Result
    Current=@n1.Transform.Translation
   ) {
    is=struct RigVMUnitNode
    template="SpringInterp::Execute(in Target,in Strength,in CriticalDamping,in Force,in bUseCurrentInput,in Current,in TargetVelocityAmount,in bInitializeFromTarget,out Result,out Velocity)"
   }
   &RigUnit_ModifyTransforms(
    ExecutePin=()
    ItemToModify=((Item=(Type=Control,Name="leg_ik_l"),Transform=(Rotation=(X=0,Y=0,Z=0,W=1),Translation=(X=0,Y=0,Z=0),Scale3D=(X=1,Y=1,Z=1))))
    Mode=OverrideGlobal
    Weight=1
    WeightMinimum=0
    ItemToModify.0.Transform.Rotation=@n1.Transform.Rotation
    ItemToModify.0.Transform.Scale3D=@n1.Transform.Scale3D
    ItemToModify.0.Transform.Translation=@SpringInterp.Result
   )
   > IKTwoBone = RigVMFunctionReferenceNode(
    Bone A=(Type=Control,Name="upperarm_01_r_ctrl")
    Bone B=(Type=Control,Name="lowerarm_01_r_ctrl")
    Bone C=(Type=Control,Name="hand_r_ctrl")
    Enable Stretch=false
    End Ctrl=(Type=Control,Name="leg_ik_r")
    Line Thickness=0.25
    Pole Vector Node=(Type=Control,Name="leg_ik_pole_r")
    Primary Axis=(X=1,Y=0,Z=0)
    Secondary Axis=(X=0,Y=0,Z=-1)
    Stretch Maximum Ratio=1.25
    Stretch Start Ratio=0.75
    debug=true
    scale=(X=0,Y=0,Z=0)
   )
   > IKTwoBone_1 = RigVMFunctionReferenceNode(
    Bone A=(Type=Control,Name="upperarm_01_l_ctrl")
    Bone B=(Type=Control,Name="lowerarm_01_l_ctrl")
    Bone C=(Type=Control,Name="hand_l_ctrl")
    Enable Stretch=false
    End Ctrl=(Type=Control,Name="leg_ik_l")
    Line Thickness=0.25
    Pole Vector Node=(Type=Control,Name="leg_ik_l")
    Primary Axis=(X=-1,Y=0,Z=0)
    Secondary Axis=(X=0,Y=0,Z=1)
    Stretch Maximum Ratio=1.25
    Stretch Start Ratio=0.75
    debug=true
    scale=(X=0,Y=0,Z=0)
   )
   > &RigUnit_ModifyTransforms(
    ItemToModify="((Item=(Type=Bone,Name=\"upperarm_01_r\"),Transform=(Rotation=(X=0,Y=0,Z=0,W=1),Translation=(X=0,Y=0,Z=0),Scale3D=(X=1,Y=1,Z=1))),(Item=(Type=Bone,Name=\"lowerarm_01_r\"),Transform=(Rotation=(X=0,Y=0,Z=0,W=1),Translation=(X=0,Y=0,Z=0),Scale3D=(X=1,Y=1,Z=1))),( ...(truncated)"
    Weight=1
    WeightMinimum=0
    ItemToModify.0.Transform=&RigUnit_GetTransform(Item=(Type=Control,Name="upperarm_01_r_ctrl"), Space=LocalSpace, bInitial=False).Transform
    ItemToModify.1.Transform=&RigUnit_GetTransform(Item=(Type=Control,Name="lowerarm_01_r_ctrl"), Space=LocalSpace, bInitial=False).Transform
    ItemToModify.2.Transform=&RigUnit_GetTransform(Item=(Type=Control,Name="hand_r_ctrl"), Space=LocalSpace, bInitial=False).Transform
   )
   > &RigUnit_ModifyTransforms(
    ItemToModify="((Item=(Type=Bone,Name=\"upperarm_01_l\"),Transform=(Rotation=(X=0,Y=0,Z=0,W=1),Translation=(X=0,Y=0,Z=0),Scale3D=(X=1,Y=1,Z=1))),(Item=(Type=Bone,Name=\"lowerarm_01_l\"),Transform=(Rotation=(X=0,Y=0,Z=0,W=1),Translation=(X=0,Y=0,Z=0),Scale3D=(X=1,Y=1,Z=1))),( ...(truncated)"
    Weight=1
    WeightMinimum=0
    ItemToModify.2.Transform=&RigUnit_GetTransform(Item=(Type=Control,Name="hand_l_ctrl"), Space=LocalSpace, bInitial=False).Transform
    ItemToModify.0.Transform=&RigUnit_GetTransform(Item=(Type=Control,Name="upperarm_01_l_ctrl"), Space=LocalSpace, bInitial=False).Transform
    ItemToModify.1.Transform=&RigUnit_GetTransform(Item=(Type=Control,Name="lowerarm_01_l_ctrl"), Space=LocalSpace, bInitial=False).Transform
   )
   RigUnit_BeginExecution = &RigUnit_BeginExecution()
   Remap = &RigVMFunction_MathDoubleRemap(
    SourceMaximum=75
    SourceMinimum=80
    TargetMaximum=1
    TargetMinimum=0
    Value=&RigVMFunction_MathTransformMakeRelative(Global=RigVMVariableNode() {variable=TPelvis}.Value, Parent=RigVMVariableNode() {variable=TRoot}.Value).Local.Translation.Z
   )
   Add_3 = RigVMAggregateNode(
    A=@Remap.Result
    B=&RigVMFunction_MathDoubleRemap(SourceMaximum=750, SourceMinimum=550, TargetMaximum=1, TargetMinimum=0, Value=RigVMVariableNode() {variable=ForwardVelocity}.Value).Result
   )
   RigVMFunction_MathBoolNot = &RigVMFunction_MathBoolNot(Value=RigVMVariableNode() {variable=isSkyDiving}.Value)
   RigVMFunction_AlphaInterp = &RigVMFunction_AlphaInterp(
    InterpSpeedDecreasing=5
    InterpSpeedIncreasing=5
    Scale=1
    bClampResult=False
    bInterpResult=True
    Value=&RigVMFunction_MathDoubleMul(A=@Add_3.Result, B=&RigVMFunction_MathBoolToFloat(Value=@RigVMFunction_MathBoolNot.Result).Result).Result
   )
   &RigUnit_ModifyTransforms(
    ItemToModify="((Item=(Type=Bone,Name=\"upperarm_01_l\"),Transform=(Rotation=(X=-0.2,Y=0.3736,Z=0.0828,W=0.902),Translation=(X=0,Y=0,Z=0),Scale3D=(X=1,Y=1,Z=1))),(Item=(Type=Bone,Name=\"lowerarm_01_l\"),Transform=(Rotation=(X=0,Y=-0.342,Z=0,W=0.9397),Translation=(X=0,Y=0,Z=0 ...(truncated)"
    WeightMinimum=0
    ExecutePin=@RigUnit_BeginExecution.ExecutePin
    Weight=@RigVMFunction_AlphaInterp.Result
   )
   Cast = /Script/RigVM.RigVMFunction_MathQuaternionToRotator(Value=&RigVMFunction_MathTransformMakeRelative(Global=RigVMVariableNode() {variable=TSpine03}.Value, Parent=RigVMVariableNode() {variable=TPelvis}.Value).Local.Rotation) {is=struct RigVMUnitNode, template="Cast::Execute(in Value,out Result)"}
   DISPATCH_RigVMDispatch_BreakStruct = RigVMDispatchNode(Struct=@Cast.Result) {template="DISPATCH_RigVMDispatch_BreakStruct(in Struct,out Elements)"}
   > &RigUnit_ModifyTransforms(
    ItemToModify=((Item=(Type=Bone,Name="pack"),Transform=(Rotation=(X=-0.1305,Y=0,Z=0,W=0.9914),Translation=(X=0,Y=0,Z=5),Scale3D=(X=1,Y=1,Z=1))))
    WeightMinimum=0
    Weight=&RigVMFunction_MathDoubleRemap(SourceMaximum=-30, SourceMinimum=0, TargetMaximum=1, TargetMinimum=0, Value=@DISPATCH_RigVMDispatch_BreakStruct.Elements.Yaw).Result
   )
   > &RigUnit_ModifyTransforms(
    ItemToModify=((Item=(Type=Bone,Name="pack"),Transform=(Rotation=(X=-0.0872,Y=0,Z=0,W=0.9962),Translation=(X=0,Y=0,Z=0),Scale3D=(X=1,Y=1,Z=1))))
    WeightMinimum=0
    Weight=&RigVMFunction_MathDoubleRemap(SourceMaximum=45, SourceMinimum=0, TargetMaximum=1, TargetMinimum=0, Value=@DISPATCH_RigVMDispatch_BreakStruct.Elements.Yaw).Result
   )
   > &RigUnit_ModifyTransforms(
    ItemToModify="((Item=(Type=Bone,Name=\"clavicle_01_l\"),Transform=(Rotation=(X=0.1736,Y=0,Z=0,W=0.9848),Translation=(X=0,Y=0,Z=0),Scale3D=(X=1,Y=1,Z=1))),(Item=(Type=Bone,Name=\"clavicle_01_r\"),Transform=(Rotation=(X=0.1736,Y=0,Z=0,W=0.9848),Translation=(X=0,Y=0,Z=0),Scale ...(truncated)"
    WeightMinimum=0
    Weight=&RigVMFunction_MathDoubleRemap(SourceMaximum=-60, SourceMinimum=-30, TargetMaximum=1, TargetMinimum=0, Value=@DISPATCH_RigVMDispatch_BreakStruct.Elements.Yaw).Result
   )
   /Script/ControlRig.RigUnit_SetTranslation(
    ExecutePin=()
    Item=(Type=Control,Name="leg_ik_l")
    Space=LocalSpace
    Value=(X=-7398085658.2656,Y=0,Z=0)
    Weight=1
    bInitial=False
    bPropagateToChildren=True
   ) {is=struct RigVMUnitNode, template="Set Transform::Execute(in Item,in Space,in bInitial,in Value,in Weight,in bPropagateToChildren)"}
   n2 = &RigUnit_GetTransform(Item=(Type=Control,Name="leg_ik_r"), Space=LocalSpace, bInitial=False)
   > &RigUnit_SetTranslation(
    Item=(Type=Control,Name="leg_ik_r")
    Space=LocalSpace
    Weight=1
    bInitial=False
    bPropagateToChildren=True
    Value=&RigUnit_SpringInterpVectorV2(
     CriticalDamping=1
     Strength=4
     TargetVelocityAmount=0
     bInitializeFromTarget=False
     bUseCurrentInput=False
     Target=&RigVMFunction_MathVectorAdd(A=(X=0,Y=0,Z=0), B=@n2.Transform.Translation) {template="Add::Execute(in A,in B,out Result)"}.Result
     Force=/Script/RigVM.RigVMFunction_MathVectorSub(B=(X=0,Y=0,Z=0), A=@n2.Transform.Translation) {is=struct RigVMUnitNode, template="Subtract::Execute(in A,in B,out Result)"}.Result
     Current=@n2.Transform.Translation
    ) {template="SpringInterp::Execute(in Target,in Strength,in CriticalDamping,in Force,in bUseCurrentInput,in Current,in TargetVelocityAmount,in bInitializeFromTarget,out Result,out Velocity)"}.Result
   ) {template="Set Transform::Execute(in Item,in Space,in bInitial,in Value,in Weight,in bPropagateToChildren)"}
   &RigVMFunction_MathVectorAdd(A=(X=0,Y=0,Z=0), B=@n1.Transform.Translation) {template="Add::Execute(in A,in B,out Result)"}
   ComputePoleVector = RigVMFunctionReferenceNode(
    Bone A=(Type=Bone,Name="None")
    Bone B=(Type=Bone,Name="None")
    Bone C=(Type=Bone,Name="None")
    Draw Transform=true
    ExecuteContext=()
    Initial=false
    OffsetFactor=2
    Origin Scale=(X=0,Y=0,Z=0)
   )
   &RigVMFunction_MathDoubleRemap(SourceMaximum=1600, SourceMinimum=1000, TargetMaximum=0, TargetMinimum=1, Value=RigVMVariableNode() {variable=UpVelocity}.Value)
   n3 = &RigVMFunction_MathTransformMakeRelative(Global=RigVMVariableNode() {variable=TPelvis}.Value, Parent=RigVMVariableNode() {variable=TRoot}.Value)
   &RigVMFunction_MathDoubleRemap(
    SourceMaximum=-15
    SourceMinimum=70
    TargetMaximum=-1
    TargetMinimum=0
    Value=RigVMDispatchNode(Struct=&RigVMFunction_MathQuaternionToRotator(Value=@n3.Local.Rotation) {template="Cast::Execute(in Value,out Result)"}.Result) {template="DISPATCH_RigVMDispatch_BreakStruct(in Struct,out Elements)"}.Elements.Pitch
   )
   RigUnit_GetTransform = &RigUnit_GetTransform(Item=(Type=Control,Name="leg_ik_l"), bInitial=False)
   &RigUnit_GetTransform(Item=(Type=Control,Name="leg_ik_r"), bInitial=False)
   &RigUnit_GetTransform(Item=(Type=Control,Name="leg_ik_l"), bInitial=False)
   &RigUnit_GetTransform(Item=(Type=Control,Name="ik_pole_l"), bInitial=False)
   n4 = &RigUnit_GetTransform(Item=(Type=Control,Name="leg_ik_r"), bInitial=False)
   n5 = &RigUnit_GetTransform(Item=(Type=Bone,Name="clavicle_r"), bInitial=false)
   RigVMFunction_MathVectorDistance = &RigVMFunction_MathVectorDistance(A=@n4.Transform.Translation, B=@n5.Transform.Translation)
   Divide = &RigVMFunction_MathDoubleDiv(B=2, A=@RigVMFunction_MathVectorDistance.Result)
   Subtract = &RigVMFunction_MathVectorSub(A=@n4.Transform.Translation, B=@n5.Transform.Translation) {template="Subtract::Execute(in A,in B,out Result)"}
   RigVMFunction_MathVectorSetLength = /Script/RigVM.RigVMFunction_MathVectorSetLength(Length=@Divide.Result, Value=@Subtract.Result) {is=struct RigVMUnitNode, template="SetLength::Execute(in Value,in Length,out Result)"}
   Add = &RigVMFunction_MathVectorAdd(B=(X=-3,Y=0,Z=-30), A=@RigVMFunction_MathVectorSetLength.Result) {template="Add::Execute(in A,in B,out Result)"}
   RigVMFunction_VisualDebugVectorNoSpace = /Script/RigVM.RigVMFunction_VisualDebugVectorNoSpace(
    Color=(R=1,G=0,B=0,A=1)
    DebugDrawSettings=(DepthPriority=SDPG_Foreground,Lifetime=-1)
    Mode=Vector
    Scale=1
    Thickness=1
    bEnabled=True
    Value=@Add.Result
   ) {is=struct RigVMUnitNode, template="VisualDebug::Execute(io Value,in bEnabled,in Mode,in Color,in Thickness,in Scale,in DebugDrawSettings)"}
   RigUnit_SetTranslation = &RigUnit_SetTranslation(
    ExecutePin=()
    Item=(Type=Control,Name="leg_ik_pole_r")
    Space=GlobalSpace
    Weight=1
    bInitial=False
    bPropagateToChildren=true
    Value=@RigVMFunction_VisualDebugVectorNoSpace.Value
   ) {template="Set Transform::Execute(in Item,in Space,in bInitial,in Value,in Weight,in bPropagateToChildren)"}
   TimeOffset = /Script/RigVM.RigVMFunction_TimeOffsetTransform(BufferSize=8, SecondsAgo=0.1, TimeRange=1, Value=(Rotation=(X=0,Y=0,Z=0,W=1),Translation=(X=0,Y=0,Z=0),Scale3D=(X=1,Y=1,Z=1))) {is=struct RigVMUnitNode, template="TimeOffset::Execute(in Value,in SecondsAgo,in BufferSize,in TimeRange,out Result)"}
   /Script/ControlRig.RigUnit_TwoBoneIKSimplePerItem(
    DebugSettings=(bEnabled=False,Scale=10,WorldOffset=(Rotation=(X=0,Y=0,Z=0,W=1),Translation=(X=0,Y=0,Z=0),Scale3D=(X=1,Y=1,Z=1)))
    Effector=(Rotation=(X=0,Y=0,Z=0,W=1),Translation=(X=0,Y=0,Z=0),Scale3D=(X=1,Y=1,Z=1))
    EffectorItem=(Type=Bone,Name="None")
    ExecutePin=()
    ItemA=(Type=Bone,Name="None")
    ItemALength=0
    ItemB=(Type=Bone,Name="None")
    ItemBLength=0
    PoleVector=(X=0,Y=0,Z=1)
    PoleVectorKind=Direction
    PoleVectorSpace=(Type=Bone,Name="None")
    PrimaryAxis=(X=1,Y=0,Z=0)
    SecondaryAxis=(X=0,Y=1,Z=0)
    SecondaryAxisWeight=1
    StretchMaximumRatio=1.25
    StretchStartRatio=0.75
    Weight=1
    bEnableStretch=False
    bPropagateToChildren=True
   ) {is=struct RigVMUnitNode}
   RigVMVariableNode() {variable=TPack}
  ]}
]
hierarchy [
 root {loc=<0 0 0>, rot=<0 0 0 1>}
 pack {parent=root, loc=<0 -0.7316 0>, rot=<0 0 0 1>}
 body {parent=pack, loc=<0 -5.1686 -1.5481>, rot=<0 0 0 1>}
 leg_ik_l {controlType=EULER_TRANSFORM, parent=body, type=Control, loc=<0 0 0>, rot=<0 0 0 1>}
 leg_ik_pole_l {controlType=EULER_TRANSFORM, parent=leg_ik_l, type=Control, loc=<0 0 0>, rot=<0 0 0 1>}
 leg_ik_r {controlType=EULER_TRANSFORM, parent=body, type=Control, loc=<0 0 0>, rot=<0 0 0 1>}
 leg_ik_pole_r {controlType=EULER_TRANSFORM, parent=leg_ik_r, type=Control, loc=<0 0 0>, rot=<0 0 0 1>}
 clavicle_01_r {parent=body, loc=<-5.872 0 -1.1021>, rot=<-0.0114 0.0002 0.9998 0.0138>}
 clavicle_02_r {parent=clavicle_01_r, loc=<2.1662 0 0>, rot=<0.414 -0.5514 -0.4107 -0.5966>}
 upperarm_01_r {parent=clavicle_02_r, loc=<3.74 0 0>, rot=<-0.0194 0.3017 -0.0317 -0.9527>}
 upperarm_02_r {parent=upperarm_01_r, loc=<13.6391 0 0>, rot=<0.0395 0.4236 -0.0185 -0.9048>}
 lowerarm_01_r {parent=upperarm_02_r, loc=<2.4872 0 0>, rot=<0.0709 0.6052 0.0998 -0.7866>}
 lowerarm_02_r {parent=lowerarm_01_r, loc=<23.4854 0 0>, rot=<-0.0095 -0.4537 -0.0092 -0.8911>}
 hand_r {parent=lowerarm_02_r, loc=<4.5285 0 0>, rot=<-0.0114 -0.7526 0.0199 -0.6581>}
 limb_end_r {parent=hand_r, loc=<33.3776 0 0>, rot=<0.1498 0.6728 0.712 -0.134>}
 clavicle_01_l {parent=body, loc=<5.872 0 -1.1021>, rot=<0.0138 -0.9998 0.0002 0.0114>}
 clavicle_02_l {parent=clavicle_01_l, loc=<-2.1662 0 0>, rot=<0.414 -0.5514 -0.4107 -0.5966>}
 upperarm_01_l {parent=clavicle_02_l, loc=<-3.74 0 0>, rot=<-0.0194 0.3017 -0.0317 -0.9527>}
 upperarm_02_l {parent=upperarm_01_l, loc=<-13.6391 0 0>, rot=<0.0395 0.4236 -0.0185 -0.9048>}
 lowerarm_01_l {parent=upperarm_02_l, loc=<-2.4872 0 0>, rot=<0.0709 0.6052 0.0998 -0.7866>}
 lowerarm_02_l {parent=lowerarm_01_l, loc=<-23.4854 0.0001 0>, rot=<-0.0095 -0.4537 -0.0092 -0.8911>}
 hand_l {parent=lowerarm_02_l, loc=<-4.5286 0 0>, rot=<-0.0114 -0.7526 0.0199 -0.6581>}
 limb_end_l {parent=hand_l, loc=<-33.3776 0 0.0001>, rot=<0.1498 0.6728 0.712 -0.134>}
 upperarm_01_r_ctrl {controlType=EULER_TRANSFORM, type=Control, loc=<0 0 0>, rot=<0 0 0 1>}
 lowerarm_01_r_ctrl {controlType=EULER_TRANSFORM, parent=upperarm_01_r_ctrl, type=Control, loc=<0 0 0>, rot=<0 0 0 1>}
 hand_r_ctrl {controlType=EULER_TRANSFORM, parent=lowerarm_01_r_ctrl, type=Control, loc=<0 0 0>, rot=<0 0 0 1>}
 upperarm_01_l_ctrl {controlType=EULER_TRANSFORM, type=Control, loc=<0 0 0>, rot=<0 0 0 1>}
 lowerarm_01_l_ctrl {controlType=EULER_TRANSFORM, parent=upperarm_01_l_ctrl, type=Control, loc=<0 0 0>, rot=<0 0 0 1>}
 hand_l_ctrl {controlType=EULER_TRANSFORM, parent=lowerarm_01_l_ctrl, type=Control, loc=<0 0 0>, rot=<0 0 0 1>}
]
variables [
 ForwardVelocity {type=double, default=0}
 PawnVelocity {type=FVector, default=(X=0,Y=0,Z=0)}
 TPack {type=FTransform, default=(Rotation=(X=0,Y=0,Z=0,W=1),Translation=(X=0,Y=0,Z=0),Scale3D=(X=1,Y=1,Z=1))}
 TPelvis {type=FTransform, default=(Rotation=(X=0,Y=0,Z=0,W=1),Translation=(X=0,Y=0,Z=0),Scale3D=(X=1,Y=1,Z=1))}
 TRoot {type=FTransform, default=(Rotation=(X=0,Y=0,Z=0,W=1),Translation=(X=0,Y=0,Z=0),Scale3D=(X=1,Y=1,Z=1))}
 TSpine03 {type=FTransform, default=(Rotation=(X=0,Y=0,Z=0,W=1),Translation=(X=0,Y=0,Z=0),Scale3D=(X=1,Y=1,Z=1))}
 UpVelocity {type=double, default=0}
 isSkyDiving {type=bool, default=False}
]
