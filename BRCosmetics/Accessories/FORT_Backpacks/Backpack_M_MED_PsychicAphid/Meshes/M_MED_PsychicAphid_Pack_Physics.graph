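# Physics asset for the PsychicAphid backpack cosmetic mesh, defining simulation bodies for its pair of arm-like bone chains.
# Each side has a sphere at clavicle_02 and capsules on the upper arm, lower arm and hand, mirrored left to right.
# Constraints link each chain from hand to clavicle with locked twist and tight swing limits, loosest at the hand, so the limbs flex only slightly.
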PhysicsAsset /BRCosmetics/Accessories/FORT_Backpacks/Backpack_M_MED_PsychicAphid/Meshes/M_MED_PsychicAphid_Pack_Physics
bodies {
 clavicle_02_l [{shapeType=Sphere, radius=1.8095, shapeName=clavicle_02_l_sphere}]
 clavicle_02_r [{shapeType=Sphere, radius=1.8095, shapeName=clavicle_02_l_sphere}]
 hand_l [{
   shapeType=Capsule
   length=30.2129
   radius=3.4137
   shapeName=hand_l_capsule
   center=<-16.0208 0 0.2012>
   rotation=<rot -89.2804 -179.9998 179.9997>
  }]
 hand_r [{
   shapeType=Capsule
   length=30.2129
   radius=3.4137
   shapeName=hand_l_capsule
   center=<16.0208 0 -0.2012>
   rotation=<rot -89.2804 -179.9998 -0.0003>
  }]
 lowerarm_01_l [{
   shapeType=Capsule
   length=24.8096
   radius=3.0678
   shapeName=lowerarm_01_l_capsule
   center=<-10.3183 0 0.0077>
   rotation=<rot 89.6106 0.002 0.002>
  }]
 lowerarm_01_r [{
   shapeType=Capsule
   length=24.8096
   radius=3.0678
   shapeName=lowerarm_01_l_capsule
   center=<10.3183 0 -0.0077>
   rotation=<rot 89.6106 0.002 -179.998>
  }]
 upperarm_01_l [{
   shapeType=Capsule
   length=11.7975
   radius=2.167
   shapeName=upperarm_01_l_capsule
   center=<-6.0807 0 -0.0109>
   rotation=<rot 89.8969 179.9893 179.9894>
  }]
 upperarm_01_r [{
   shapeType=Capsule
   length=11.7975
   radius=2.167
   shapeName=upperarm_01_l_capsule
   center=<6.0807 0 0.0109>
   rotation=<rot 89.8969 179.9893 -0.0106>
  }]
}
constraints [
 {bone1Name=hand_l, bone2Name=lowerarm_01_l, swing1LimitDegrees=10, swing2LimitDegrees=10, twistLimitDegrees=45, twistMotion=Locked}
 {bone1Name=lowerarm_01_l, bone2Name=upperarm_01_l, swing1LimitDegrees=3, swing2LimitDegrees=8, twistLimitDegrees=45, twistMotion=Locked}
 {bone1Name=upperarm_01_l, bone2Name=clavicle_02_l, swing1LimitDegrees=3, swing2LimitDegrees=3, twistLimitDegrees=45, twistMotion=Locked}
 {bone1Name=upperarm_01_r, bone2Name=clavicle_02_r, swing1LimitDegrees=3, swing2LimitDegrees=3, twistLimitDegrees=45, twistMotion=Locked}
 {bone1Name=lowerarm_01_r, bone2Name=upperarm_01_r, swing1LimitDegrees=3, swing2LimitDegrees=8, twistLimitDegrees=45, twistMotion=Locked}
 {bone1Name=hand_r, bone2Name=lowerarm_01_r, swing1LimitDegrees=10, swing2LimitDegrees=10, twistLimitDegrees=45, twistMotion=Locked}
]
